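; Death ability for the ToyFox boss that plays the death outro montage and kills the boss with a finisher tag.
; On activation it stops look-at targeting and the StateTree logic while montage notifies drive the rez-out cue, disable collision and destruction, and trigger Die.
; A montage-length delay and ability end also invoke death as fallbacks, with a gate on bDeathInvoked guarding against repeat kills.
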(event Ability|EventActivateAbility
  ;[0 0.15 0.08155 0.5] ActivateAbility: Play the death montage and destroy self at a certain point.
  (|CallParentFunction)
  (Fortnite|FortShippingLog self "ActivateAbility: Invoked")
  (bind (_asynctaskproxy _notifytag) (Ability|Tasks|PlayMontageAndWaitNotify "DeathMontageTask" (Variables|Settings|GetDeathOutroMontage))
    (:then
      (Utilities|FlowControl|Sequence
        (:then_0
          ;; Queen should not turn to target while playing out the death animation.
          (bind _asbga_tidal_crane_boss_toy_fox (Utilities|Casting|CastToBGA_TidalCrane_Boss_ToyFox (Ability|GetAvatarActorfromActorInfo)))
          (Class|BGATidalCraneBoss|AllowLookatTarget _asbga_tidal_crane_boss_toy_fox false))
        (:then_1
          ;;[0 0.15 0.1354 0.5] Disable 'brain'.
          (bind _statetree (Class|BPITidalCraneBoss|GetStateTree(Message) (Ability|GetAvatarActorfromActorInfo)))
          (Utilities|IsValid _statetree
            (:Is%20Valid
              (AI|Logic|StopLogic _statetree ""))
            (:Is%20Not%20Valid
              _))
          ;;[0.03293 0 0.15 0.5] Fallback to invoke death after a certain duration.
          (Utilities|FlowControl|Delay (Animation|GetPlayLength (Variables|Settings|GetDeathOutroMontage)))
          (CallFunction|InvokeDeath))))
    (:OnNotifyBegin
      ;;[0 0.15 0.0959 0.5] Play death rez-out effects here.
      (if (== _notifytag "(TagName=\"TidalCrane.ToyFox.AnimNotify.Ability.Death.RezOut\")")
        (Fortnite|FortShippingLog self "ActivateAbility: Notify rezout invoked.")
        (Ability|ExecuteGameplayCueOnOwner (Utilities|Struct|BreakGameplayCueTag (Variables|GameplayCueTag|GetGameplayCueTag_DeathRezOut)))
        (elif (== _notifytag "(TagName=\"TidalCrane.ToyFox.AnimNotify.Ability.Death.DisableCollision\")")
          ;;[0 0.15 0.02809 0.5] Disable collision on the boss as it is rez-ing out.
          (Fortnite|FortShippingLog self "ActivateAbility: Notify disable collision invoked.")
          (bind _asbga_tidal_crane_boss_toy_fox_2 (Utilities|Casting|CastToBGA_TidalCrane_Boss_ToyFox (Ability|GetAvatarActorfromActorInfo)))
          (Class|BGATidalCraneBossToyFox|SetBossCollisionEnable _asbga_tidal_crane_boss_toy_fox_2 false)
          (Class|BGATidalCraneBossToyFox|SetEnableDestructionComponent _asbga_tidal_crane_boss_toy_fox_2 false)
          (elif (== _notifytag "(TagName=\"TidalCrane.ToyFox.AnimNotify.Ability.Death.Die\")")
            ;;[0 0.08328 0.15 0.5] Immediately destroy the building actor.
            (Utilities|FlowControl|DeltaGateWildcard (Variables|Transient|GetbDeathInvoked) true
              (:Success
                (Fortnite|FortShippingLog self "ActivateAbility: Notify Die invoked.")
                (bind _asbga_tidal_crane_boss_toy_fox_3 (Utilities|Casting|CastToBGA_TidalCrane_Boss_ToyFox (Ability|GetAvatarActorfromActorInfo)))
                (Gameplay|Die _asbga_tidal_crane_boss_toy_fox_3 "(TagName=\"TidalCrane.ToyFox.Finisher\")" (Fortnite|GetFortPlayerControllerfromActor (Ability|EffectContext|GetInstigatorActor (Ability|GetGrantedbyEffectContext))) (Ability|EffectContext|GetInstigatorActor (Ability|GetGrantedbyEffectContext)) :!class /Script/FortniteGame.BuildingActor) ; Specific case as ATidalCraneSwarmBoss class :\nWe will simulate a SilentDie() in code when the tag "TidalCrane.ToyFox.Finisher" is used.
              ))))))))

(event Custom|InvokeDeath
  ;[0 0.15 0.08155 0.5] ActivateAbility: Play the death montage and destroy self at a certain point.
  ;;[0 0.08328 0.15 0.5] Immediately destroy the building actor.
  (Utilities|FlowControl|DeltaGateWildcard (Variables|Transient|GetbDeathInvoked) true
    (:Success
      (Fortnite|FortShippingLog self "ActivateAbility: Notify Die invoked.")
      (bind _asbga_tidal_crane_boss_toy_fox (Utilities|Casting|CastToBGA_TidalCrane_Boss_ToyFox (Ability|GetAvatarActorfromActorInfo)))
      (Gameplay|Die _asbga_tidal_crane_boss_toy_fox "(TagName=\"TidalCrane.ToyFox.Finisher\")" (Fortnite|GetFortPlayerControllerfromActor (Ability|EffectContext|GetInstigatorActor (Ability|GetGrantedbyEffectContext))) (Ability|EffectContext|GetInstigatorActor (Ability|GetGrantedbyEffectContext)) :!class /Script/FortniteGame.BuildingActor) ; Specific case as ATidalCraneSwarmBoss class :\nWe will simulate a SilentDie() in code when the tag "TidalCrane.ToyFox.Finisher" is used.
    )))

(event Ability|EventOnEndAbility
  (params (bWasCancelled :type Boolean))
  ;[0.08333 0.03503 0 0.5] OnEndAbility: Fallback to invoke death in case ability ends before finishing up as expected.
  (Utilities|FlowControl|Sequence
    (:then_0
      (CallFunction|InvokeDeath))
    (:then_1
      (|CallParentFunction bWasCancelled))))

(fn CanActivateAbility
  (params (ActorInfo :type Gameplay%20Ability%20Actor%20Info%20Structure) (Handle :type Gameplay%20Ability%20Spec%20Handle%20Structure))
  (return 0 true))
